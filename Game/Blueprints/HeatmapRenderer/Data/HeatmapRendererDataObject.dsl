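; Data object for a heatmap renderer that holds parallel arrays of locations and severities and returns a location and severity pair by index.
; GetSeverityRange scans severities once to cache min and max, using 99999999 as the unset sentinel; forced recaching rescans without resetting them.
; GetSeveritiesNormalized lazily fills and caches a normalized severity array from that range, reusing it once populated.
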(fn getBasicDataOutput
  (params (index :type Integer))
  (return (Utilities|Array|Get(aref) (Variables|Default|GetLocations) index) (Utilities|Array|Get(aref) (Variables|Default|GetSeverities) index)))

(fn GetSeveritiesNormalized
  (if (Utilities|Array|IsValidIndex (Variables|Default|GetNormalizedSeverities))
    (return (Variables|Default|GetNormalizedSeverities))
    (else
      (bind (_severitymin _severitymax) (CallFunction|GetSeverityRange false))
      (for _array_element (Variables|Default|GetSeverities)
        (Utilities|Array|Add (Variables|Default|GetNormalizedSeverities) (Math|Float|NormalizetoRange _array_element _severitymin _severitymax)))
      (return (Variables|Default|GetNormalizedSeverities)))))

(fn GetSeverityRange
  (params (ForceRecaching :type Boolean))
  (if (or ForceRecaching (== (Variables|Protected|GetSeverityMin) 99999999.0))
    (for _array_element (Variables|Default|GetSeverities)
      (Variables|Protected|SetSeverityMin (Math|Float|Min(Float) _array_element (Variables|Protected|GetSeverityMin)))
      (Variables|Protected|SetSeverityMax (Math|Float|Max(Float) _array_element (Variables|Protected|GetSeverityMax)))))
  (return (Variables|Protected|GetSeverityMin) (Variables|Protected|GetSeverityMax)))
